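# Niagara system for the match flame in the Keytar rocket launcher emote, including the strike and the sustained flame.
# Curl-noise flame and smoke ribbons rise in spawn groups, with flame width and lifetime adjusted by LOD distance.
# Layered glow sprites redden and shrink with age, while flipbook smoke sprites add detail.
# A brief spark burst and scaling mesh flash form the strike, pulled toward the camera by offsets.
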NiagaraSystem /SparksCosmetics/Emotes/Keytar/RocketLauncher/Effects/NS_Emote_Keytar_RocketLaunch_MatchFlame_Ribbon
dependencies {
 usedDataInterfaces [
  &NiagaraDataInterfaceCurve
  &NiagaraDataInterfaceVector2DCurve
  &NiagaraDataInterfaceSpriteRendererInfo
  &NiagaraDataInterfaceColorCurve
 ]
 usedDynamicInputs [&Vector2DFromCurve, &RandomRangeVector2D]
 usedModules [&SystemState]
}
emitters {
 FlameRibbon {
  emitterUpdateScript [
   &EmitterState()
   &SpawnRate(SpawnRate=30, Use Spawn Probability=false)
  ]
  particleSpawnScript [
   &FN_SetSpawnGroup(Normalize Spawn Group=true, Number of Spawn Groups=11, Particles Per Group=10, Set Ribbon ID and Ribbon Link Order=true)
   &InitializeRibbon(
    Color=<rgba 10 1.75 0.2 0.3>
    Lifetime=&Multiply_Float(
     A=&Lerp_Float(
      A=0.75
      B=0.85
      Alpha=&FN_LOD_Distance_Remap(Start Distance=500, End Distance=1500, Minimum Output=0, Maximum Output=1)
     )
     B=1.25
    )
    Module.Write Color=false
    Module.Write Lifetime=true
    Module.Write Mass=true
    Module.Write Material Random=true
    Module.Write Position=true
    Module.Write RibbonTwist=false
    Module.Write RibbonWidth=true
    Ribbon Width=&Multiply_Float(
     A=&Lerp_Float(
      A=3
      B=5
      Alpha=&FN_LOD_Distance_Remap(Start Distance=500, End Distance=1500, Minimum Output=0, Maximum Output=1)
     )
     B=3
    )
   )
  ]
  particleUpdateScript [
   &ParticleState()
   &GravityForce(Gravity=<0 0 25>)
   /Niagara/Modules/Update/Forces/CurlNoiseForce(
    Noise Frequency=0.01
    Noise Strength=&FloatFromCurve(
     FloatCurve=DataInterface() {propertyValues {Curve {keys [
         {arriveTangent=0.1097, interpMode=RCIM_Cubic, leaveTangent=0.1097, tangentMode=RCTM_User, time=0, value=0}
         {arriveTangent=1.6798, interpMode=RCIM_Cubic, leaveTangent=1.6798, tangentMode=RCTM_User, time=1, value=1}
        ]}}}
     Scale Curve=15
    )
    Pan Noise=false
    Random Seed:NiagaraInt32=&RandomRangeInt(Minimum=0, Maximum=100)
   )
   &Drag(Drag=2.5, Use Linear Drag=true, Use Rotational Drag=false) {enabled=false}
   &SolveForcesAndVelocity(Clamp Velocity=false, Limit Acceleration=false)
   &Color(Color=&ColorFromCurve(ColorCurve=DataInterface() {propertyValues {
       AlphaCurve {keys [<at 0 1 RCIM_Cubic>]}
       BlueCurve {keys [<at 0 10 RCIM_Cubic>,<at 0.1638 0.2 RCIM_Cubic>]}
       GreenCurve {keys [<at 0 1.75 RCIM_Cubic>,<at 0.1638 1.75 RCIM_Cubic>]}
       RedCurve {keys [<at 0 0.2 RCIM_Cubic>,<at 0.1638 10 RCIM_Cubic>]}
      }}))
   &ScaleColor(
    Color Value To Scale=$Particles.Color
    Scale Alpha=&FloatFromCurve(
     FloatCurve=DataInterface() {propertyValues {Curve {keys [
         {arriveTangent=-0.0176, interpMode=RCIM_Cubic, leaveTangent=-0.0176, tangentMode=RCTM_User, time=0.1, value=0}
         {arriveTangent=0.1593, interpMode=RCIM_Cubic, leaveTangent=0.1593, tangentMode=RCTM_User, time=0.3, value=1}
         {arriveTangent=-0.0405, interpMode=RCIM_Cubic, leaveTangent=-0.0405, tangentMode=RCTM_User, time=1, value=0}
        ]}}}
     Scale Curve=0.5
    )
    ScaleA=true
    ScaleRGB=true
   )
   &ScaleRibbonWidth(Ribbon Width Scale=&FloatFromCurve(
     FloatCurve=DataInterface() {propertyValues {Curve {keys [
         {arriveTangent=0.0742, interpMode=RCIM_Cubic, leaveTangent=0.0742, tangentMode=RCTM_User, time=0, value=0.4}
         {arriveTangent=-0.0613, interpMode=RCIM_Cubic, leaveTangent=-0.0613, tangentMode=RCTM_User, time=0.1408, value=0.9847}
         {arriveTangent=-0.0405, interpMode=RCIM_Cubic, leaveTangent=-0.0405, tangentMode=RCTM_User, time=1, value=0}
        ]}}}
     Scale Curve=0.45
    ))
   &ScaleRibbonWidth(Ribbon Width Scale=0.5) {module=ScaleRibbonWidth001}
  ]
  rendererClasses [&NiagaraRibbonRendererProperties]
  renderers [{rendererClass=&NiagaraRibbonRendererProperties}]
 }
 GlowBaseAndOuter {
  emitterUpdateScript [
   &EmitterState()
   &SpawnRate(SpawnRate=15, Use Spawn Probability=false)
  ]
  particleSpawnScript [
   &FN_SetSpawnGroup(Normalize Spawn Group=true, Number of Spawn Groups=11, Particles Per Group=10, Set Ribbon ID and Ribbon Link Order=true)
   &InitializeParticle(
    Color=<rgba 15 10 1 1>
    Lifetime=&Multiply_Float(A=&RandomRangeFloat(Minimum=0.2, Maximum=0.5), B=1)
    Position Offset=<0 0 0.5>
    Sprite Size Mode=::NewEnumerator3
    Uniform Sprite Size=&Multiply_Float(A=&RandomRangeFloat(Minimum=0.15, Maximum=0.175), B=5.5)
    UsePositionOffset=true
    Write Material Random=true
   ) {module=InitializeParticle001}
   /Niagara/Modules/Spawn/Initialization/InitializeParticle(
    Color=<rgba 15 10 1 1>
    Lifetime=&Multiply_Float(A=&UniformRangedFloat(Minimum=0.2, Maximum=0.5), B=1)
    Module.Write Color=true
    Module.Write Lifetime=true
    Module.Write Mass=true
    Module.Write Material Random=true
    Module.Write Position=true
    Module.Write Scale=false
    Module.Write SpriteRotation=false
    Module.Write SpriteSize=true
    Position=&SimulationPosition()
    Sprite Size=&Vector2DFromFloat(Value=&Multiply_Float(A=&UniformRangedFloat(Minimum=0.65, Maximum=1), B=3))
   ) {enabled=false}
   &Color(Color=&MultiplyLinearColor(
     Linear Color=$Particles.Color
     Scale Factor=/Niagara/DynamicInputs/UniformRange/UniformRangedVector4(
      Minimum:Vector4f=<1 0.25 0.25 0.01>
      Maximum:Vector4f=<1 1 1 1>
      Randomness Mode:ENiagaraRandomnessMode=::NewEnumerator0
      Random Seed:NiagaraInt32=0
      Fixed Random Seed:NiagaraBool=false
     )
    ))
  ]
  particleUpdateScript [
   &ParticleState()
   &AddVelocity(Coordinate Space=::World, Velocity=&RandomRangeVector(Minimum=<-0.25 -0.25 1>, Maximum=<0.25 0.25 3>))
   &Drag(Drag=30, Use Linear Drag=true, Use Rotational Drag=false)
   &SolveForcesAndVelocity(Clamp Velocity=false, Limit Acceleration=false)
   &ScaleColor(
    Scale Alpha=&FloatFromCurve(FloatCurve=DataInterface() {propertyValues {Curve {keys [<at 0.5 1>,<at 1 0>]}}})
    Scale RGB=&Lerp_Vector(A=<1 1 1>, B=<1 0 0>, Alpha=$Particles.NormalizedAge)
    ScaleA=true
    ScaleRGB=true
   )
   &ScaleSpriteSize(Scale Factor=&Vector2DFromFloat(Value=&FloatFromCurve(FloatCurve=DataInterface() {propertyValues {Curve {keys [
          {arriveTangent=-0.0036, interpMode=RCIM_Cubic, leaveTangent=-0.0036, tangentMode=RCTM_User, time=0, value=1}
          {arriveTangent=-2.2201, interpMode=RCIM_Cubic, leaveTangent=-2.2201, tangentMode=RCTM_User, time=1, value=0}
         ]}}})))
   /SetVariables(
    Particles.ColorAlt:LinearColor=&MultiplyLinearColor(Linear Color=$Particles.Color, Scale Factor=<1 0.35 0.1 0.02>)
    Particles.SpriteSizeAlt:Vector2f=<1.75 2.75>
   ) {isSetParameters=true}
   &CameraOffset(Camera Offset Amount=1)
  ]
  rendererClasses [&NiagaraSpriteRendererProperties]
  renderers [
   {rendererClass=&NiagaraSpriteRendererProperties}
   {rendererClass=&NiagaraSpriteRendererProperties, rendererIndex=1}
  ]
 }
 SmokeRibbon {
  emitterUpdateScript [
   &EmitterState()
   &SpawnRate(SpawnRate=25, Use Spawn Probability=false)
  ]
  particleSpawnScript [
   &FN_SetSpawnGroup(Normalize Spawn Group=true, Number of Spawn Groups=11, Particles Per Group=10, Set Ribbon ID and Ribbon Link Order=true)
   &InitializeRibbon(
    Color=<rgba 0.0729 0.0128 0.0015 1>
    Lifetime=&Multiply_Float(
     A=&RandomRangeFloat(Minimum=0.5, Maximum=0.9, Randomness Mode=::NewEnumerator1, Random Seed=$Particles.SpawnGroup)
     B=1.5
    )
    Module.Write Color=true
    Module.Write Lifetime=true
    Module.Write Mass=true
    Module.Write Material Random=true
    Module.Write Position=true
    Module.Write RibbonTwist=false
    Module.Write RibbonWidth=true
    Ribbon Width=&Multiply_Float(A=&RandomRangeFloat(Minimum=2.5, Maximum=3.5), B=2)
   )
  ]
  particleUpdateScript [
   &ParticleState()
   &GravityForce(Gravity=<0 0 75>)
   &CurlNoiseForce(
    Noise Frequency=0.01
    Noise Strength=&FloatFromCurve(
     FloatCurve=DataInterface() {propertyValues {Curve {keys [
         {arriveTangent=1.0245, interpMode=RCIM_Cubic, leaveTangent=1.0245, tangentMode=RCTM_User, time=0, value=0}
         {arriveTangent=1.6798, interpMode=RCIM_Cubic, leaveTangent=1.6798, tangentMode=RCTM_User, time=1, value=1}
        ]}}}
     Scale Curve=30
    )
    Pan Noise=false
    Random Seed:NiagaraInt32=&RandomRangeInt(Minimum=0, Maximum=100)
   )
   &Drag(Drag=2.5, Use Linear Drag=true, Use Rotational Drag=false) {module=Drag001}
   &FN_RibbonMaterial(Color Pan U Speed=-0.5, Erosion Amount=1, Macro Distortion Amount=1, Opacity Erosion=0.25, Opacity Pan U Speed=-0.5, Opacity Power=0.35)
   &ScaleRibbonWidth(Ribbon Width Scale=&FloatFromCurve(
     FloatCurve=DataInterface() {propertyValues {Curve {keys [
         {arriveTangent=0.0219, interpMode=RCIM_Cubic, leaveTangent=0.0219, tangentMode=RCTM_User, time=0, value=0}
         <at 0.5 0.9666 RCIM_Cubic>
         {arriveTangent=-0.2116, interpMode=RCIM_Cubic, leaveTangent=-0.2116, tangentMode=RCTM_User, time=1, value=0}
        ]}}}
     Scale Curve=&RandomRangeFloat(Minimum=1, Maximum=1)
    )) {module=ScaleRibbonWidth001}
   &SolveForcesAndVelocity(Clamp Velocity=false, Limit Acceleration=false)
   &ScaleColor(
    Scale Alpha=&FloatFromCurve(
     FloatCurve=DataInterface() {propertyValues {Curve {keys [
         {arriveTangent=9.0725, interpMode=RCIM_Cubic, leaveTangent=9.0725, tangentMode=RCTM_User, time=0, value=0}
         {arriveTangent=0.0486, interpMode=RCIM_Cubic, leaveTangent=0.0486, tangentMode=RCTM_User, time=0.1408, value=0.9847}
         {arriveTangent=-0.0405, interpMode=RCIM_Cubic, leaveTangent=-0.0405, tangentMode=RCTM_User, time=1, value=0}
        ]}}}
     Scale Curve=0.75
    )
    ScaleA=true
    ScaleRGB=true
   )
   &ScaleRibbonWidth(Ribbon Width Scale=0.5)
  ]
  rendererClasses [&NiagaraRibbonRendererProperties]
  renderers [{rendererClass=&NiagaraRibbonRendererProperties}]
 }
 Smoke_Detail {
  emitterUpdateScript [
   &EmitterState(Inactive Response=::NewEnumerator1, Life Cycle Mode=::NewEnumerator1, Loop Delay=0.4, Loop Duration=1, UseLoopDelay=false)
   &SpawnBurst_Instantaneous(Spawn Count=5, Use Spawn Probability=false, UseLoopCountLimit=false)
   &SpawnRate(SpawnRate=50, Use Spawn Probability=false) {module=SpawnRate001}
  ]
  particleSpawnScript [/Niagara/Modules/Spawn/Initialization/V2/InitializeParticle(
    Color=<rgba 0.2353 0.2353 0.2353 0.325>
    Lifetime Max=0.75
    Lifetime Min=0.5
    Lifetime Mode=::NewEnumerator1
    Sprite Rotation Mode=::NewEnumerator2
    Sprite Size Max=<0.7 0.7>
    Sprite Size Min=<0.5 0.5>
    Sprite Size Mode=::NewEnumerator2
    UsePositionOffset=false
    Write Material Random=true
   )]
  particleUpdateScript [
   &ParticleState()
   &ScaleSpriteSize(Scale Sprite Size Mode=::NewEnumerator0, Uniform Scale Factor=0.25) {module=ScaleSpriteSize001}
   &SpriteRotationRate(Rotation Rate=&RandomRangeFloat(Minimum=-50, Maximum=50))
   &ScaleSpriteSize(
    Scale Sprite Size Mode=::NewEnumerator0
    Uniform Scale Factor=&FloatFromCurve(FloatCurve=DataInterface() {propertyValues {Curve {keys [<at 0 0.45>,<at 1 1.2>]}}}, Scale Curve=8)
   )
   /Niagara/Modules/Update/Forces/V2/CurlNoiseForce(Noise Frequency=0.1, Noise Strength=25, Pan Noise=true, Pan Noise Field=<0 0 1>) {enabled=false}
   &PointAttractionForce(
    Attraction Strength=0.25
    Attractor Position Offset:Vector3f=<0 0 75>
    Kill Within Radius=false
    Module.Use Falloff=true
    UseAttractorOffset=true
   )
   &SubUVAnimation(
    End Frame Range Override=30
    "Sprite Renderer "=DataInterface() {propertyValues {SpriteRenderer=&self:Smoke_Detail_1.NiagaraSpriteRendererProperties_0}}
    Start Frame Range Override=&RandomRangeInt(Minimum=0, Maximum=5)
    UseEndFrame=true
    UseStartFrame=true
   )
   &ScaleColor(
    Scale Alpha=&Multiply_Float(
     A=&RandomRangeFloat(Minimum=0.5, Maximum=1)
     B=&FloatFromCurve(FloatCurve=DataInterface() {propertyValues {Curve {keys [
          {arriveTangent=9.0725, interpMode=RCIM_Cubic, leaveTangent=9.0725, tangentMode=RCTM_User, time=0, value=0}
          {arriveTangent=-2.5249, interpMode=RCIM_Cubic, leaveTangent=-2.5249, tangentMode=RCTM_User, time=0.1408, value=0.9847}
          {arriveTangent=-0.0405, interpMode=RCIM_Cubic, leaveTangent=-0.0405, tangentMode=RCTM_User, time=1, value=0}
         ]}}})
    )
    ScaleA=true
    ScaleRGB=true
   )
   &GravityForce(Gravity=&Multiply_Vector(
     A=<1 1 150>
     B=&MakeVector(
      X=1
      Y=1
      Z=&FloatFromCurve(
       FloatCurve=DataInterface() {propertyValues {Curve {keys [
           {interpMode=RCIM_Cubic, tangentMode=RCTM_User, time=0, value=1}
           {arriveTangent=0.8127, interpMode=RCIM_Cubic, leaveTangent=0.8127, tangentMode=RCTM_User, time=1, value=0}
          ]}}}
       Scale Curve=0.1
      )
     )
    )) {enabled=false}
   &GravityForce(Gravity=<0 0 35>) {module=GravityForce001}
   &Drag(Use Linear Drag=true, Use Rotational Drag=true)
   &SolveForcesAndVelocity(Clamp Velocity=false, Limit Acceleration=false)
   &CameraOffset(Camera Offset Amount=2)
  ]
  rendererClasses [&NiagaraSpriteRendererProperties]
  renderers [{rendererClass=&NiagaraSpriteRendererProperties}]
 }
 SparkBurst {
  emitterUpdateScript [
   &EmitterState(Life Cycle Mode=::NewEnumerator1, Loop Behavior=::NewEnumerator1, Loop Delay=0.05, Loop Duration=1, UseLoopDelay=true)
   &SpawnBurst_Instantaneous(Spawn Count=15, Use Spawn Probability=false)
  ]
  particleSpawnScript [
   &InitializeParticle(
    Color=<rgba 1 0.2157 0.0667 1>
    Lifetime Max=0.15
    Lifetime Min=0.05
    Lifetime Mode=::NewEnumerator1
    Position Offset=<0 0 5>
    Sprite Size Max=<0.4 0.8>
    Sprite Size Min=<0.2 0.4>
    Sprite Size Mode=::NewEnumerator2
    UsePositionOffset=false
    Write Material Random=true
   )
   &ShapeLocation(MaskSpawnGroup=false, Override Seed=false, Sphere Radius=1, Sphere Surface Distribution=1, UseOwnerScale=true)
   &AddVelocity(
    Module.Use Falloff Distance=false
    Origin Offset=<0 0 -0.1>
    Override Seed=false
    Velocity Mode=::NewEnumerator1
    Velocity Speed=&RandomRangeFloat(Minimum=50, Maximum=75)
   )
  ]
  particleUpdateScript [
   &ParticleState()
   &ScaleColor(
    Scale Alpha=&FNTimeOfDayScalar(Night Scale=1, Morning Scale=1, Evening Scale=1)
    Scale RGB=&VectorFromFloat(Value=&FloatFromCurve(FloatCurve=DataInterface() {propertyValues {Curve {keys [<at 0 10>,<at 1 1>]}}}))
    ScaleA=true
    ScaleRGB=true
   ) {module=ScaleColor001}
   &Drag(Drag=&RandomRangeFloat(Minimum=8, Maximum=20), Use Linear Drag=true, Use Rotational Drag=false)
   &CurlNoiseForce(Noise Strength=&RandomRangeFloat(Minimum=200, Maximum=400), Pan Noise=true)
   &GravityForce(Gravity=&RandomRangeVector(Minimum=<0 0 0>, Maximum=<0 0 -100>)) {enabled=false}
   &SolveForcesAndVelocity(Clamp Velocity=false, Limit Acceleration=false)
   &ScaleSpriteSizeBySpeed(Max Scale Factor=<1 10>, Min Scale Factor=<0.5 0.1>)
   &ScaleSpriteSize(Scale Factor=&MakeVector2D(
     X=&FloatFromCurve(FloatCurve=DataInterface() {propertyValues {Curve {keys [<at 0 0.5>,<at 0.1 1>,<at 1 0.5>]}}}, Scale Curve=3)
     Y=&FloatFromCurve(FloatCurve=DataInterface() {propertyValues {Curve {keys [<at 0 1>,<at 1 0>]}}}, Scale Curve=8)
    ))
   &CameraOffset(Camera Offset Amount=10)
  ]
  rendererClasses [&NiagaraSpriteRendererProperties]
  renderers [{rendererClass=&NiagaraSpriteRendererProperties}]
 }
 StrikeSource {
  emitterUpdateScript [
   &EmitterState(
    Inactive Response=::NewEnumerator1
    Life Cycle Mode=::NewEnumerator1
    Loop Behavior=::NewEnumerator1
    Loop Delay=&UniformRangedFloat(Minimum=0.5, Maximum=2.5)
    Loop Duration=0.25
    UseLoopDelay=false
   )
   &SpawnBurst_Instantaneous(Spawn Count=3, Use Spawn Probability=false)
  ]
  particleSpawnScript [
   &InitializeParticle(
    Color=<rgba 1 0.4314 0.1333 1>
    Lifetime=0.075
    Mesh Scale Mode=::NewEnumerator3
    Mesh Uniform Scale=0.1
    UsePositionOffset=false
    Write Material Random=true
   ) {module=InitializeParticle001}
   &InitialMeshOrientation(Mesh Orientation Mode=::NewEnumerator2, Orientation Vector=<0 0 -1>, Use Rotation Vector=false) {module=InitialMeshOrientation001}
  ]
  particleUpdateScript [
   &ParticleState()
   &SolveForcesAndVelocity(Clamp Velocity=false, Limit Acceleration=false)
   &ScaleMeshSize(Scale Factor=&VectorFromFloat(Value=&FloatFromCurve(FloatCurve=DataInterface() {propertyValues {Curve {keys [{arriveTangent=2.0877, interpMode=RCIM_Cubic, leaveTangent=2.0877, tangentMode=RCTM_User, time=0, value=0}, <at 1 1>]}}})))
   &CameraOffset(Camera Offset Amount=10) {module=CameraOffset001}
  ]
  rendererClasses [&NiagaraMeshRendererProperties]
  renderers [{rendererClass=&NiagaraMeshRendererProperties, rendererIndex=0}]
 }
}
summary {
 systemName=NS_Emote_Keytar_RocketLaunch_MatchFlame_Ribbon
 emitters [
  {emitterName=SparkBurst, rendererClasses [&NiagaraSpriteRendererProperties]}
  {emitterName=StrikeSource, rendererClasses [&NiagaraMeshRendererProperties]}
  {emitterName=Smoke_Detail, rendererClasses [&NiagaraSpriteRendererProperties]}
  {emitterName=GlowBaseAndOuter, rendererClasses [&NiagaraSpriteRendererProperties]}
  {emitterName=SmokeRibbon, rendererClasses [&NiagaraRibbonRendererProperties]}
  {emitterName=FlameRibbon, rendererClasses [&NiagaraRibbonRendererProperties]}
 ]
 userVariables [User.Interpolate Intensity {type=NiagaraFloat, defaultValue=0.5, description=""}]
}
