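# Niagara smoke detail effect for the rocket launch in the Sparks Keytar RocketLauncher emote.
# A single sprite emitter bursts short-lived brown-grey puffs with random size, spin and velocity along positive Z.
# Heavy drag brakes the initial launch while curl noise, wind and upward gravity keep the smoke drifting.
# Sub-UV animation plays as puffs grow, darken and fade, with spin slowing on a curve driven by emitter age.
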NiagaraSystem /SparksCosmetics/Emotes/Keytar/RocketLauncher/Effects/NS_Emote_Keytar_RocketLaunch_SmokeDetail_A
dependencies {usedDataInterfaces [&NiagaraDataInterfaceCurve], usedDynamicInputs [], usedModules [&SystemState]}
emitters {Smoke_Launch_Detail {
  emitterUpdateScript [
   &EmitterState(Life Cycle Mode=::NewEnumerator1, Loop Behavior=::NewEnumerator1, Loop Duration=1, UseLoopDelay=false)
   &SpawnBurst_Instantaneous(Spawn Count=40, Use Spawn Probability=false, UseLoopCountLimit=false)
  ]
  particleSpawnScript [
   &InitializeParticle(
    Color=<rgba 0.3 0.21 0.18 0.6>
    Lifetime=&UniformRangedFloat(Minimum=1, Maximum=2)
    Mass Mode=::NewEnumerator1
    Position=/Niagara/DynamicInputs/Helpers/SimulationPosition()
    Position Mode=::NewEnumerator1
    Sprite Rotation Angle=&AngleConversion(Angle:NiagaraFloat=0, Angle Input:ENiagara_AngleInput=::NewEnumerator0, Angle Output:ENiagara_AngleInput=::NewEnumerator1)
    Sprite Rotation Mode=::NewEnumerator0
    Sprite Size=&Vector2DFromFloat(Value=&UniformRangedFloat(Minimum=80, Maximum=130))
    Sprite Size Mode=::NewEnumerator4
    UsePositionOffset=false
    Write Material Random=true
   ) {module=InitializeParticle001}
   /SetVariables(Particles.InitSpriteRotRate=&UniformRangedFloat(Minimum=-3, Maximum=3)) {isSetParameters=true}
   &SystemLocation(MaskSpawnGroup=false, Offset=<-20 0 0>)
   &AddVelocity(Velocity=&RandomRangeVector(Minimum=<0 0 0>, Maximum=<0 0 2500>)) {module=AddVelocity001}
  ]
  particleUpdateScript [
   &ParticleState()
   &SpriteRotationRate(Rotation Rate=&Multiply_Float(
     A=$Particles.InitSpriteRotRate
     B=&FloatFromCurve(
      FloatCurve=DataInterface() {propertyValues {Curve {keys [
          {arriveTangent=-1.9214, interpMode=RCIM_Cubic, leaveTangent=-1.9214, tangentMode=RCTM_User, time=-0.0046, value=1.0113}
          {arriveTangent=-0.2496, interpMode=RCIM_Cubic, leaveTangent=-0.2496, tangentMode=RCTM_User, time=0.9896, value=0.16}
         ]}}}
      CurveIndex=$Emitter.Age
      Scale Curve=150
     )
    ))
   &ScaleColor(
    Scale Alpha=&FloatFromCurve(
     FloatCurve=DataInterface() {propertyValues {Curve {keys [<at 0 0>, {interpMode=RCIM_Cubic, leaveTangent=-2.2891, tangentMode=RCTM_Break, time=0.2, value=1}, <at 1 0>]}}}
     Scale Curve=0.5
    )
    Scale RGB=&VectorFromFloat(Value=&FloatFromCurve(FloatCurve=DataInterface() {propertyValues {Curve {keys [<at 0 1>,<at 1 0.5>]}}}))
    ScaleA=true
    ScaleRGB=true
   )
   &CurlNoiseForce(Noise Strength=100, Pan Noise=false)
   &WindForce(Wind Speed Scale=&RandomRangeFloat(Minimum=0.5, Maximum=1))
   &Drag(Drag=7, Use Linear Drag=true, Use Rotational Drag=true) {module=Drag001}
   &GravityForce(Gravity=<0 0 100>)
   &SolveForcesAndVelocity(Clamp Velocity=false, Limit Acceleration=false)
   &ScaleSpriteSize(Scale Factor=&Vector2DFromFloat(Value=&FloatFromCurve(FloatCurve=DataInterface() {propertyValues {Curve {keys [
          {arriveTangent=2.9298, interpMode=RCIM_Cubic, leaveTangent=0.4137, tangentMode=RCTM_User, time=0, value=0.25}
          {arriveTangent=-0.0186, interpMode=RCIM_Cubic, leaveTangent=1.2864, tangentMode=RCTM_User, time=1, value=1.2}
         ]}}})))
   &SubUVAnimation(End Frame=32)
  ]
  rendererClasses [&NiagaraSpriteRendererProperties]
  renderers [{rendererClass=&NiagaraSpriteRendererProperties}]
 }}
summary {
 systemName=NS_Emote_Keytar_RocketLaunch_SmokeDetail_A
 emitters [{emitterName=Smoke_Launch_Detail, rendererClasses [&NiagaraSpriteRendererProperties]}]
 userVariables []
}
